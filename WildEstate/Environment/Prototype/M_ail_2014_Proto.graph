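# M_ail_2014_Proto is a Default Lit prototype environment material in the WildEstate content.
# Its graph holds a single Color vector parameter defaulting to dark grey, with no textures, math or visible link to a material output.
# It is flagged for use on spline meshes, with automatic usage setting in the editor turned off.
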Material /WildEstate/Environment/Prototype/M_ail_2014_Proto
graph {
 properties {bAutomaticallySetUsageInEditor=false, bUsedWithSplineMeshes=true, usedShadingModels=Default Lit}
 nodes [VectorParameter() {parameterName=Color, defaultValue=<rgba 0.1458 0.1458 0.1458 0>}]
}
